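; Server-driven pawn for the FossilMeal_Bread NPC boss, which can be hired and manages leap-attack state.
; On spawn it links its animation layers, binds alert and hire events, and sets a radial leash if no patrol path exists.
; Alert changes toggle extra weapons and leaping, while damage dealt by the hiring pawn sets a timed leap target.
; It also scales incoming damage via hotfix values, finds random hostile pawns, and shows a duel-defeat bubble on death.
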(event EventBeginPlay
  (bind _controllerref (Variables|Default|GetControllerRef))
  (|CallParentFunction)
  (Components|SkeletalMesh|AnimationBlueprintLinking|LinkAnimClassLayers (Variables|Character|GetMesh) "/FossilMeal_Bread/Assets/Animations/FossilMeal_Bread_AnimBP.FossilMeal_Bread_AnimBP_C")
  (bind _asfort_athena_aibot_controller (Utilities|Casting|CastToFortAthenaAIBotController (Pawn|GetController)))
  (Variables|Default|SetControllerRef _asfort_athena_aibot_controller)
  (Networking|SwitchHasAuthority
    (:Authority
      (Variables|Default|SetSpawnLocation (Transformation|GetActorLocation))
      (AlertLevel|BindEventtoOnAlertLevelChangedEvent (Variables|Default|GetControllerRef) (EventDispatchers|CreateEvent :!event Custom|AlertLevelUpdated))
      (ConvertEvents|BindEventtoOnConvertedEvent (Variables|FortPlayerPawn|GetConvertComponent) (EventDispatchers|CreateEvent :!event Custom|OnHired))
      (Utilities|FlowControl|Delay 1.0)
      (Utilities|IsValid (Class|FortAthenaNpcPatrollingComponent|GetPatrolPath (Variables|/Script/FortniteAI.FortAthenaAIBotController|GetCachedPatrollingComponent _controllerref) :!class /Script/FortniteAI.FortAthenaNpcPatrollingComponent)
        (:Is%20Not%20Valid
          (Class|BPFLAICommonFunctions|SetRadialLeash _controllerref (Transformation|GetActorLocation) 8e+03 1.5e+04))))))

(event Custom|OnHired
  (params (InstigatorPawn :type Fort%20Pawn%20Object%20Reference) (ConvertedPawn :type Fort%20Pawn%20Object%20Reference))
  (Utilities|IsValid InstigatorPawn
    (:Is%20Valid
      (bind _output_get (Variables|Default|SetHireeRef InstigatorPawn))
      (EventDispatchers|BindEventtoOnDidDamage _output_get (EventDispatchers|CreateEvent :!event Custom|HireeDidDamage)))))

(event Custom|HireeDidDamage
  (params (DamagedActor :type Actor%20Object%20Reference) (Damage :type Float%20%28single-precision%29) (InstigatedBy :type Controller%20Object%20Reference) (DamageCauser :type Actor%20Object%20Reference) (HitLocation :type Vector) (FHitComponent :type Primitive%20Component%20Object%20Reference) (BoneName :type Name) (Momentum :type Vector))
  (Utilities|IsValid (Variables|Default|GetLeapTarget)
    (:Is%20Not%20Valid
      (bind _asfort_pawn (Utilities|Casting|CastToFortPawn DamagedActor))
      (Variables|Default|SetLeapTarget _asfort_pawn)
      (Variables|Default|SetCanLeap true)
      (CallFunction|StartForgetTimer))))

(event Custom|StartForgetTimer
  (Utilities|Time|ClearandInvalidateTimerbyHandle (Variables|Default|GetLeapTargetForget_Th))
  (bind _returnvalue (Utilities|Time|SetTimerbyEvent (delegate Custom|ForgetLeapTarget) (ScalableFloat|GetValueatLevel (Variables|Hotfix|GetLeapTargetForgetTime_Hf))))
  (Variables|Default|SetLeapTargetForget_Th _returnvalue))

(event Custom|ForgetLeapTarget
  (Variables|Default|SetLeapTarget 0))

(event Custom|AlertLevelUpdated
  (params (BotController :type Fort%20Athena%20AIBot%20Controller%20Object%20Reference) (OldAlertLevel :type EAlertLevel%20Enum :default "Unaware") (NewAlertLevel :type EAlertLevel%20Enum :default "Unaware"))
  (bind _allowadditionalweapons_hf (Variables|Hotfix|GetAllowAdditionalWeapons_Hf))
  (bind _returnvalue (ScalableFloat|GetValueasBool _allowadditionalweapons_hf))
  (bind _controllerref (Variables|Default|GetControllerRef))
  (bind _additionalweapons (Variables|Default|GetAdditionalWeapons))
  (switch Utilities|FlowControl|Switch|SwitchonEAlertLevel NewAlertLevel
    (:Unaware
      (Variables|Default|SetCanLeap false)
      (Variables|Default|SetLeapTarget 0)
      (Variables|Default|SetEmergencyLeptOnce false)
      (if _returnvalue
        (for _array_element _additionalweapons
          (Inventory|RemoveItemfromInventoryOwner _controllerref _array_element))))
    (:Suspicious
      (if _returnvalue
        (for _array_element_2 _additionalweapons
          (Inventory|GiveItemtoInventoryOwner _controllerref _array_element_2))))
    (:Threatened
      (Utilities|IsValid (Variables|Default|GetHireeRef)
        (:Is%20Not%20Valid
          (Variables|Default|SetCanLeap true))))))

(event EventOnDeathServer
  (params (Damage :type Float%20%28single-precision%29) (DamageTags :type Gameplay%20Tag%20Container%20Structure%20%28by%20ref%29) (Momentum :type Vector :default "0, 0, 0") (HitInfo :type Hit%20Result%20Structure%20%28by%20ref%29) (InstigatedBy :type Controller%20Object%20Reference) (DamageCauser :type Actor%20Object%20Reference) (EffectContext :type Gameplay%20Effect%20Context%20Handle%20Structure))
  (|CallParentFunction Damage DamageTags Momentum HitInfo InstigatedBy DamageCauser EffectContext)
  (if (GameplayTags|HasMatchingGameplayTag "(TagName=\"Pawn.Athena.NPC.Duel.IsBeingChallenged\")")
    (Class|BPFLAICommonFunctions|PlaySpeechBubble self "(TagName=\"Athena.AI.SpeechBubbles.FossilMealBread.DuelDefeat\")" true true)))

(fn BlueprintModifyIncomingDamage
  (params (Damage :type Float%20%28single-precision%29) (DamageTags :type Gameplay%20Tag%20Container%20Structure%20%28by%20ref%29) (EffectContext :type Gameplay%20Effect%20Context%20Handle%20Structure%20%28by%20ref%29) (EventInstigator :type Controller%20Object%20Reference) (DamageCauser :type Actor%20Object%20Reference) (HitInfo :type Hit%20Result%20Structure))
  (bind _returnvalue (|CallParentFunction Damage DamageTags EffectContext EventInstigator DamageCauser HitInfo))
  (return (* _returnvalue (Math|Float|SelectFloat (ScalableFloat|GetValueatLevel (Variables|Hotfix|GetReducedDmgIncPrc_Hf)) 1.0 (Variables|Default|GetReduceIncDmg)))))

(fn FindRandomHostileFortPawn
  (bind _hireeref (Variables|Default|GetHireeRef))
  (bind _outfortpawns (Fortnite|GetFortPawnsInRange (Transformation|GetActorLocation) (ScalableFloat|GetValueatLevel (Variables|Hotfix|GetFingTargetRange_Hf))))
  (bind _output_get (LocalVariables|SetFoundPawns _outfortpawns))
  (for _array_element _output_get
    (if (== _array_element _hireeref)
      _
      (elif (Utilities|Enum|Equal(Enum) (Team|GetTeamAffiliation _array_element _hireeref) "Hostile")
        (Utilities|Array|AddUnique (LocalVariables|GetFilteredPawns) _array_element))))
  (return (Utilities|Array|RandomArrayItem (LocalVariables|GetFilteredPawns))))

(fn PawnPlayAbilityAnim
  (params (Montage :type Anim%20Montage%20Object%20Reference) (PlayRate :type Float%20%28double-precision%29) (SectionName :type Name))
  (Animation|PlayAnimMontage Montage PlayRate SectionName)
  (bind _returnvalue (ContextualAnim|Utilities|GetSectionLength (Variables|GetMontage) (Montage|GetSectionIndex Montage SectionName)))
  (return (* _returnvalue (Variables|GetPlayRate))))

(fn ConstructionScript
  (|CallParentFunction))
